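# Niagara effect for the Steep emote's tea pour, drawn as a ribbon stream falling from a spout offset.
# TeaStream, the only enabled emitter, spawns ribbon points with velocity shaped over loop age, then bends them with gravity and local acceleration.
# Nested curves over particle age, loop progress and a spawn counter taper the width so the pour swells and thins out.
# The Drip and two Syrup_Ribbons emitters are disabled.
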NiagaraSystem /BRCosmetics/Effects/Fort_Effects/Emotes/Steep/NS_Emote_Steep_Pour
dependencies {
 usedDataInterfaces [
  &NiagaraDataInterfaceCurve
  &NiagaraDataInterfaceVectorCurve
  /Script/Niagara.NiagaraDataInterfaceSimpleCounter
  &NiagaraDataInterfaceColorCurve
 ]
 usedDynamicInputs []
 usedModules [
  &SystemState
  &self:SystemSpawnScript.SystemScriptSource.SystemScriptGraph.NiagaraNodeAssignment_0.SetVariables_#1
 ]
}
emitters {
 Drip {
  bEnabled=false
  emitterUpdateScript [
   &EmitterState()
   &SpawnRate(
    SpawnRate=&RandomRangeFloat(Minimum=1, Maximum=16, Evaluation Type=::NewEnumerator1)
    Use Spawn Probability=false
   )
   /SetVariables(Emitter.PlayerVelocityVector:Vector3f=$Engine.Owner.Velocity) {enabled=false, isSetParameters=true}
   /SetVariables(Emitter.PlayerSpeed:NiagaraFloat=&VectorLength(Vector=$Emitter.PlayerVelocityVector)) {enabled=false, isSetParameters=true}
   &SpawnBurst_Instantaneous(Spawn Count=1000, Use Spawn Probability=false)
  ]
  particleSpawnScript [
   &InitializeParticle(
    Lifetime=0.27
    Mesh Scale=&Multiply_VectorByFloat(Vector=<0.6 0.6 0.6>, Float=&RandomRangeFloat(Minimum=0.75, Maximum=1.3))
    Mesh Scale Mode=::NewEnumerator4
    Position Offset=&Add_Vector(A=$System.SpoutOffset, B=<0 0 -4>)
    Position Offset Coordinate Space=::Local
    Sprite Size Mode=::NewEnumerator1
    Uniform Sprite Size Max=5
    Uniform Sprite Size Min=3
    UsePositionOffset=true
    Write Material Random=true
   ) {module=InitializeParticle001}
   &ApplyInitialForces()
   &InheritVelocity(Module.Limit Speed=true, Velocity Scale=&Multiply_VectorByFloat(Vector=<1 1 1>, Float=-0.1)) {enabled=false, module=InheritVelocity001}
  ]
  particleUpdateScript [
   &ParticleState()
   &GravityForce(Gravity=&Multiply_VectorByFloat(
     Vector=<1 1 -700>
     Float=&FloatFromCurve(FloatCurve=DataInterface() {propertyValues {Curve {keys [{arriveTangent=1.8039, interpMode=RCIM_Cubic, leaveTangent=1.8039, tangentMode=RCTM_User, time=0, value=0}, <at 1 1>]}}})
    ))
   &SolveForcesAndVelocity(Clamp Velocity=false, Limit Acceleration=false)
   &ScaleMeshSize(Scale Factor=&Multiply_Vector(
     A=<1 1 1>
     B=&VectorFromCurve(VectorCurve=DataInterface() {propertyValues {
        XCurve {keys [
          <at 0 0.5 RCIM_Cubic>
          {arriveTangent=0.5, interpMode=RCIM_Cubic, leaveTangent=-5.3087, tangentMode=RCTM_Break, time=0.1222, value=2.3}
          <at 1 0.5>
         ]}
        YCurve {keys [<at 0 1>,<at 1 1>]}
        ZCurve {keys [<at 0 1>,<at 1 1>]}
       }})
    ))
   &DynamicMaterialParameters(
    Index 0 Param 1=0.8
    Index 0 Param 2=&FloatFromCurve(FloatCurve=DataInterface() {propertyValues {Curve {keys [{arriveTangent=-1.8884, interpMode=RCIM_Cubic, leaveTangent=-1.8884, tangentMode=RCTM_User, time=0, value=1}, <at 1 0.26>]}}})
    Index 0 Param 3=1
    Index 0 Param 4=1
    Module.Param0WriteEnabled=true
    Module.Param1WriteEnabled=true
    Module.Param2WriteEnabled=true
    Module.Param3WriteEnabled=true
   )
  ]
  rendererClasses [&NiagaraSpriteRendererProperties, &NiagaraMeshRendererProperties]
  renderers [
   {rendererClass=&NiagaraSpriteRendererProperties}
   {rendererClass=&NiagaraMeshRendererProperties, rendererIndex=1}
   {rendererClass=&NiagaraMeshRendererProperties, rendererIndex=2}
  ]
 }
 Syrup_Ribbons001 {
  bEnabled=false
  emitterUpdateScript [
   &EmitterState(
    Inactive Response=::NewEnumerator1
    Life Cycle Mode=::NewEnumerator1
    Loop Behavior=::NewEnumerator1
    Loop Delay=0.25
    Loop Duration=1
    UseLoopDelay=true
   )
   &SpawnRate(SpawnRate=20, Use Spawn Probability=false)
  ]
  particleSpawnScript [
   /Niagara/Modules/Spawn/Initialization/InitializeParticle(
    Module.Write Color=true
    Module.Write Lifetime=true
    Module.Write Mass=true
    Module.Write Material Random=true
    Module.Write Position=false
    Module.Write Scale=false
    Module.Write SpriteRotation=false
    Module.Write SpriteSize=true
    Sprite Size=&Vector2DFromFloat(Value=25)
   ) {enabled=false}
   /Niagara/Modules/Spawn/Initialization/V2/InitializeParticle(Sprite Size Mode=::NewEnumerator3, Uniform Sprite Size=1, UsePositionOffset=false, Write Material Random=true) {module=InitializeParticle001}
   /SetVariables(Particles.RibbonWidth=10) {isSetParameters=true}
  ]
  particleUpdateScript [
   &ParticleState()
   &Drag(Drag=0.9, Use Linear Drag=true, Use Rotational Drag=false)
   &GravityForce()
   &SolveForcesAndVelocity(Clamp Velocity=false, Limit Acceleration=false)
   &FN_UpdateRibbon(
    Enable Skeletal Position=true
    Initial Color:LinearColor=$Particles.Initial.Color
    Initial RibbonWidth:NiagaraFloat=$Particles.Initial.RibbonWidth
    Sampled Position Coordinate Space:ENiagaraCoordinateSpace=::Local
    Sampled Rotation Coordinate Space:ENiagaraCoordinateSpace=::Local
    Sampled Velocity Coordinate Space:ENiagaraCoordinateSpace=::Local
   ) {enabled=false}
   &Material(Base Texture Time:NiagaraFloat=&Multiply_Float(A=$Emitter.Age, B=-1))
   &DynamicMaterialParameters(
    Index 0 Param 1=&MaskFloatBySpawnGroup(Masked Float=0, Passthrough Float=1)
    Index 0 Param 2=&FloatFromCurve(FloatCurve=DataInterface() {propertyValues {Curve {keys [{arriveTangent=-2.1032, interpMode=RCIM_Cubic, leaveTangent=-2.1032, tangentMode=RCTM_User, time=0, value=1}, <at 1 0.26>]}}})
    Module.Param0WriteEnabled=true
    Module.Param1WriteEnabled=true
    Module.Param2WriteEnabled=false
    Module.Param3WriteEnabled=false
   )
   &ScaleRibbonWidth(Ribbon Width Scale=&FloatFromCurve(FloatCurve=DataInterface() {propertyValues {Curve {keys [{arriveTangent=-5.074, interpMode=RCIM_Cubic, leaveTangent=-5.074, tangentMode=RCTM_User, time=0.0058, value=1}, <at 0.25 0.5>]}}}))
   /SetVariables(Particles.RibbonLinkOrder:NiagaraFloat=$Particles.NormalizedAge) {isSetParameters=true, module=SetVariables_#2}
  ]
  rendererClasses [&NiagaraRibbonRendererProperties, &NiagaraSpriteRendererProperties]
  renderers [
   {rendererClass=&NiagaraRibbonRendererProperties}
   {rendererClass=&NiagaraSpriteRendererProperties, rendererIndex=1}
  ]
 }
 Syrup_Ribbons002 {
  bEnabled=false
  emitterSpawnScript [/SetVariables(Emitter.RibbonCount=1, Emitter.RibbonSpawnRate=20) {isSetParameters=true, module=SetVariables_#3}]
  emitterUpdateScript [
   &EmitterState(
    Inactive Response=::NewEnumerator1
    Life Cycle Mode=::NewEnumerator1
    Loop Behavior=::NewEnumerator1
    Loop Delay=0.15
    Loop Duration=1.2
    UseLoopDelay=true
   )
   &SpawnRate(SpawnRate=$Emitter.RibbonSpawnRate, Use Spawn Probability=false)
   /SetVariables() {enabled=false, isSetParameters=true}
   &SpawnBurst_Instantaneous(Spawn Count=1, Spawn Group=-1, Use Spawn Probability=false) {enabled=false}
  ]
  particleSpawnScript [
   /Game/Effects/Niagara/Modules/Ribbons/FN_InitializeRibbon_v5(Custom Offset=true, Lifetime=1, Local Vector Offset:Vector3f=$System.SpoutOffset, Offset Amount=1, Ribbon Width=10)
   /SetVariables(Particles.SpriteSize=&Vector2DFromFloat(Value=1)) {isSetParameters=true, module=SetVariables_#4}
   /SetVariables(Particles.Counter:NiagaraFloat=&Multiply_Float(
     A=&MakeFloatFromInt(Int=/Niagara/DynamicInputs/DataInterface/GenerateIntFromCounter(SimpleCounter:NiagaraDataInterfaceSimpleCounter=DataInterface() {propertyValues {GpuSyncMode=None, InitialValue=0}}))
     B=&Divide_Float(A=1, B=$Emitter.RibbonSpawnRate)
    )) {isSetParameters=true, module=SetVariables_#5}
   &AddVelocity(Velocity=&MakeVector(
     X=&FloatFromCurve(
      FloatCurve=DataInterface() {propertyValues {Curve {keys [
          {arriveTangent=1.7154, interpMode=RCIM_Cubic, leaveTangent=1.7154, tangentMode=RCTM_User, time=0, value=0.5}
          {arriveTangent=-1.3301, interpMode=RCIM_Cubic, leaveTangent=-1.3301, tangentMode=RCTM_User, time=0.5, value=1}
          {arriveTangent=-0.017, interpMode=RCIM_Cubic, leaveTangent=-0.017, tangentMode=RCTM_User, time=1, value=0}
         ]}}}
      CurveIndex=$Emitter.NormalizedLoopAge
      Scale Curve=30
     )
     Y=&FloatFromCurve(
      FloatCurve=DataInterface() {propertyValues {Curve {keys [
          {interpMode=RCIM_Cubic, tangentMode=RCTM_User, time=0, value=0}
          {arriveTangent=2.6795, interpMode=RCIM_Cubic, leaveTangent=2.6795, tangentMode=RCTM_User, time=1, value=1}
         ]}}}
      CurveIndex=$Emitter.NormalizedLoopAge
      Scale Curve=-10
     )
     Z=0
    ))
   &AddVelocity(Velocity=&Multiply_VectorByFloat(
     Vector=<30 0 0>
     Float=&FloatFromCurve(
      FloatCurve=DataInterface() {propertyValues {Curve {keys [
          {arriveTangent=1.7154, interpMode=RCIM_Cubic, leaveTangent=1.7154, tangentMode=RCTM_User, time=0, value=0.5}
          {arriveTangent=-1.3301, interpMode=RCIM_Cubic, leaveTangent=-1.3301, tangentMode=RCTM_User, time=0.5, value=1}
          {arriveTangent=-0.017, interpMode=RCIM_Cubic, leaveTangent=-0.017, tangentMode=RCTM_User, time=1, value=0}
         ]}}}
      CurveIndex=$Emitter.NormalizedLoopAge
     )
    )) {enabled=false, module=AddVelocity001}
  ]
  particleUpdateScript [
   &ParticleState()
   &GravityForce(Gravity=<0 0 -75>)
   &AccelerationForce(Acceleration=<10 0 0>, Coordinate Space=::Local)
   &SolveForcesAndVelocity(Clamp Velocity=false, Limit Acceleration=false)
   &FN_UpdateRibbon(
    Enable Skeletal Position=true
    Initial Color:LinearColor=$Particles.Initial.Color
    Initial RibbonWidth:NiagaraFloat=$Particles.Initial.RibbonWidth
    Sampled Position Coordinate Space:ENiagaraCoordinateSpace=::Local
    Sampled Rotation Coordinate Space:ENiagaraCoordinateSpace=::Local
    Sampled Velocity Coordinate Space:ENiagaraCoordinateSpace=::Local
   )
   &Material(Base Texture Time:NiagaraFloat=&Multiply_Float(A=$Emitter.Age, B=-1)) {enabled=false}
   &DynamicMaterialParameters(
    Index 0 Param 1=&RandomRangeFloat(Minimum=0, Maximum=1)
    Index 0 Param 2=&FloatFromCurve(FloatCurve=DataInterface() {propertyValues {Curve {keys [{interpMode=RCIM_Cubic, tangentMode=RCTM_User, time=0, value=1}, {interpMode=RCIM_Cubic, tangentMode=RCTM_User, time=1, value=0}]}}})
    Module.Param0WriteEnabled=true
    Module.Param1WriteEnabled=true
    Module.Param2WriteEnabled=false
    Module.Param3WriteEnabled=false
   )
   &ScaleRibbonWidth(Ribbon Width Scale=&FloatFromCurve(
     FloatCurve=DataInterface() {propertyValues {Curve {keys [
         {arriveTangent=3.1073, interpMode=RCIM_Cubic, leaveTangent=3.1073, leaveTangentWeight=0.3288, tangentMode=RCTM_User, time=0.1, value=0}
         {
          arriveTangent=-1.6935
          arriveTangentWeight=0.0333
          interpMode=RCIM_Cubic
          leaveTangent=-1.6935
          leaveTangentWeight=0.1333
          tangentMode=RCTM_User
          time=0.6
          value=1
         }
         {interpMode=RCIM_Cubic, tangentMode=RCTM_User, time=1, value=0.5}
        ]}}}
     Scale Curve=&FloatFromCurve(
      FloatCurve=DataInterface() {propertyValues {Curve {keys [
          {arriveTangent=5.911, interpMode=RCIM_Cubic, leaveTangent=5.911, tangentMode=RCTM_User, time=0.25, value=0}
          {interpMode=RCIM_Cubic, tangentMode=RCTM_User, time=0.4, value=1}
          {arriveTangentWeight=0.2359, interpMode=RCIM_Cubic, leaveTangentWeight=0.118, tangentMode=RCTM_User, time=1, value=1}
          {arriveTangent=-4.1926, interpMode=RCIM_Cubic, leaveTangent=-4.1926, tangentMode=RCTM_User, time=1.2, value=0}
         ]}}}
      CurveIndex=$Emitter.Age
      Scale Curve=&FloatFromCurve(
       FloatCurve=DataInterface() {propertyValues {Curve {keys [
           {arriveTangent=7.6038, interpMode=RCIM_Cubic, leaveTangent=7.6038, tangentMode=RCTM_User, time=0, value=0.35}
           {interpMode=RCIM_Cubic, tangentMode=RCTM_User, time=0.2, value=1}
           {arriveTangentWeight=0.4719, interpMode=RCIM_Cubic, leaveTangentWeight=0.059, tangentMode=RCTM_User, time=0.9, value=1}
           {arriveTangent=-4.1926, interpMode=RCIM_Cubic, leaveTangent=-4.1926, tangentMode=RCTM_User, time=1, value=0}
          ]}}}
       CurveIndex=$Particles.Counter
      )
     )
    ))
   &ScaleRibbonWidth(Ribbon Width Scale=&FloatFromCurve(
     FloatCurve=DataInterface() {propertyValues {Curve {keys [
         {arriveTangent=4.0469, interpMode=RCIM_Cubic, leaveTangent=4.0469, leaveTangentWeight=0.3288, tangentMode=RCTM_User, time=0, value=0}
         {
          arriveTangent=-1.6935
          arriveTangentWeight=0.0333
          interpMode=RCIM_Cubic
          leaveTangent=-1.6935
          leaveTangentWeight=0.1333
          tangentMode=RCTM_User
          time=0.25
          value=1
         }
         {interpMode=RCIM_Cubic, tangentMode=RCTM_User, time=1, value=0.5}
        ]}}}
     Scale Curve=&FloatFromCurve(
      FloatCurve=DataInterface() {propertyValues {Curve {keys [
          {arriveTangent=5.911, interpMode=RCIM_Cubic, leaveTangent=5.911, tangentMode=RCTM_User, time=0.25, value=0}
          {interpMode=RCIM_Cubic, tangentMode=RCTM_User, time=0.4, value=1}
          {arriveTangent=-1.4599, interpMode=RCIM_Cubic, leaveTangent=-1.4599, tangentMode=RCTM_User, time=0.8, value=0.85}
          {arriveTangent=-4.1926, interpMode=RCIM_Cubic, leaveTangent=-4.1926, tangentMode=RCTM_User, time=1, value=0}
         ]}}}
      CurveIndex=$Emitter.Age
     )
    )) {enabled=false, module=ScaleRibbonWidth001}
   /SetVariables(Particles.RibbonLinkOrder:NiagaraFloat=$Particles.NormalizedAge) {isSetParameters=true, module=SetVariables_#2}
   &Color(Color=&ColorFromCurve(
     ColorCurve=DataInterface() {propertyValues {
       AlphaCurve {keys [<at 0 1 RCIM_Cubic>]}
       BlueCurve {keys [<at 0 0 RCIM_Cubic>,<at 0.5 0>,<at 1 1 RCIM_Cubic>]}
       GreenCurve {keys [<at 0 0 RCIM_Cubic>,<at 0.5 1>,<at 1 0 RCIM_Cubic>]}
       RedCurve {keys [<at 0 1 RCIM_Cubic>,<at 0.5 0>,<at 1 0 RCIM_Cubic>]}
      }}
     CurveIndex=$Emitter.Age
    ))
  ]
  rendererClasses [&NiagaraRibbonRendererProperties, &NiagaraSpriteRendererProperties]
  renderers [
   {rendererClass=&NiagaraRibbonRendererProperties}
   {rendererClass=&NiagaraSpriteRendererProperties, rendererIndex=1}
  ]
 }
 TeaStream {
  emitterSpawnScript [/SetVariables(Emitter.RibbonCount=1, Emitter.RibbonSpawnRate=15) {isSetParameters=true, module=SetVariables_#3}]
  emitterUpdateScript [
   &EmitterState(
    Inactive Response=::NewEnumerator1
    Life Cycle Mode=::NewEnumerator1
    Loop Behavior=::NewEnumerator1
    Loop Delay=0.1
    Loop Duration=1.3
    UseLoopDelay=true
   )
   &SpawnRate(SpawnRate=$Emitter.RibbonSpawnRate, Use Spawn Probability=false)
  ]
  particleSpawnScript [
   &FN_InitializeRibbon_v5(Custom Offset=true, Lifetime=1, Local Vector Offset:Vector3f=$System.SpoutOffset, Offset Amount=1, Ribbon Width=10)
   /SetVariables(Particles.SpriteSize=&Vector2DFromFloat(Value=1)) {isSetParameters=true, module=SetVariables_#4}
   /SetVariables(Particles.Counter:NiagaraFloat=&Multiply_Float(
     A=&MakeFloatFromInt(Int=&GenerateIntFromCounter(SimpleCounter:NiagaraDataInterfaceSimpleCounter=DataInterface() {propertyValues {GpuSyncMode=None, InitialValue=0}}))
     B=&Divide_Float(A=1, B=$Emitter.RibbonSpawnRate)
    )) {isSetParameters=true, module=SetVariables_#5}
   &AddVelocity(Velocity=&MakeVector(
     X=&FloatFromCurve(
      FloatCurve=DataInterface() {propertyValues {Curve {keys [
          {arriveTangent=1.7154, interpMode=RCIM_Cubic, leaveTangent=1.7154, tangentMode=RCTM_User, time=0, value=0.5}
          {arriveTangentWeight=0.1667, interpMode=RCIM_Cubic, leaveTangentWeight=0.1, tangentMode=RCTM_User, time=0.5, value=1}
          {arriveTangent=-4.772, arriveTangentWeight=0.1, interpMode=RCIM_Cubic, leaveTangent=-4.772, tangentMode=RCTM_User, time=0.8, value=0.3}
         ]}}}
      CurveIndex=$Emitter.NormalizedLoopAge
      Scale Curve=40
     )
     Y=&FloatFromCurve(
      FloatCurve=DataInterface() {propertyValues {Curve {keys [
          {interpMode=RCIM_Cubic, tangentMode=RCTM_User, time=0, value=0}
          {arriveTangent=2.6795, interpMode=RCIM_Cubic, leaveTangent=2.6795, tangentMode=RCTM_User, time=1, value=1}
         ]}}}
      CurveIndex=$Emitter.NormalizedLoopAge
      Scale Curve=-10
     )
     Z=0
    ))
  ]
  particleUpdateScript [
   &ParticleState()
   /SetVariables(
    Particles.NormalizeLoop:NiagaraFloat=&Divide_Float(A=&Subtract_Float(A=$Emitter.Age, B=$Emitter.CurrentLoopDelay), B=$Emitter.CurrentLoopDuration)
    Particles.RibbonLinkOrder:NiagaraFloat=$Particles.NormalizedAge
   ) {isSetParameters=true, module=SetVariables_#2}
   &GravityForce(Gravity=<0 0 -75>)
   &AccelerationForce(
    Acceleration=&Add_Vector(
     A=<10 0 0>
     B=&MakeVector(
      X=&FloatFromCurve(
       FloatCurve=DataInterface() {propertyValues {Curve {keys [
           {arriveTangent=-2.2093, interpMode=RCIM_Cubic, leaveTangent=-2.2093, tangentMode=RCTM_User, time=0, value=1}
           {interpMode=RCIM_Cubic, tangentMode=RCTM_User, time=0.3, value=-1}
           <at 0.5 0>
          ]}}}
       Scale Curve=300
      )
      Y=0
      Z=0
     )
    )
    Coordinate Space=::Local
   )
   &SolveForcesAndVelocity(Clamp Velocity=false, Limit Acceleration=false)
   &FN_UpdateRibbon(
    Enable Skeletal Position=true
    Initial Color:LinearColor=$Particles.Initial.Color
    Initial RibbonWidth:NiagaraFloat=$Particles.Initial.RibbonWidth
    Sampled Position Coordinate Space:ENiagaraCoordinateSpace=::Local
    Sampled Rotation Coordinate Space:ENiagaraCoordinateSpace=::Local
    Sampled Velocity Coordinate Space:ENiagaraCoordinateSpace=::Local
   )
   &ScaleRibbonWidth(Ribbon Width Scale=&FloatFromCurve(
     FloatCurve=DataInterface() {propertyValues {Curve {keys [
         {arriveTangent=0.9284, interpMode=RCIM_Cubic, leaveTangent=0.9284, leaveTangentWeight=0.3288, tangentMode=RCTM_User, time=0, value=0}
         {arriveTangentWeight=0.2281, interpMode=RCIM_Cubic, leaveTangentWeight=0.3422, tangentMode=RCTM_User, time=0.4, value=1}
         {arriveTangent=-2.3281, interpMode=RCIM_Cubic, leaveTangent=-2.3281, tangentMode=RCTM_User, time=1, value=0}
        ]}}}
     Scale Curve=&FloatFromCurve(
      FloatCurve=DataInterface() {propertyValues {Curve {keys [
          {arriveTangent=3.8911, interpMode=RCIM_Cubic, leaveTangent=3.8911, tangentMode=RCTM_User, time=0, value=0}
          {interpMode=RCIM_Cubic, tangentMode=RCTM_User, time=0.2, value=1}
          {arriveTangentWeight=0.2359, interpMode=RCIM_Cubic, leaveTangentWeight=0.118, tangentMode=RCTM_User, time=0.8, value=1}
          {arriveTangent=-4.1926, interpMode=RCIM_Cubic, leaveTangent=-4.1926, tangentMode=RCTM_User, time=1, value=0}
         ]}}}
      CurveIndex=$Particles.NormalizeLoop
      Scale Curve=&FloatFromCurve(
       FloatCurve=DataInterface() {propertyValues {Curve {keys [
           {arriveTangent=5.7124, interpMode=RCIM_Cubic, leaveTangent=5.7124, tangentMode=RCTM_User, time=0, value=0}
           {interpMode=RCIM_Cubic, tangentMode=RCTM_User, time=0.4, value=1}
           {arriveTangentWeight=0.4719, interpMode=RCIM_Cubic, leaveTangentWeight=0.059, tangentMode=RCTM_User, time=0.9, value=1}
           {arriveTangent=-4.1926, interpMode=RCIM_Cubic, leaveTangent=-4.1926, tangentMode=RCTM_User, time=1, value=0}
          ]}}}
       CurveIndex=$Particles.Counter
      )
     )
    ))
   &Color(Color=&ColorFromCurve(ColorCurve=DataInterface() {propertyValues {
       AlphaCurve {keys [<at 0 1 RCIM_Cubic>]}
       BlueCurve {keys [<at 0 0 RCIM_Cubic>,<at 0.5 0>,<at 1 1 RCIM_Cubic>]}
       GreenCurve {keys [<at 0 0 RCIM_Cubic>,<at 0.5 1>,<at 1 0 RCIM_Cubic>]}
       RedCurve {keys [<at 0 1 RCIM_Cubic>,<at 0.5 0>,<at 1 0 RCIM_Cubic>]}
      }})) {enabled=false}
  ]
  rendererClasses [&NiagaraRibbonRendererProperties, &NiagaraSpriteRendererProperties]
  renderers [
   {rendererClass=&NiagaraRibbonRendererProperties}
   {rendererClass=&NiagaraSpriteRendererProperties, rendererIndex=1}
   {rendererClass=&NiagaraRibbonRendererProperties, rendererIndex=2}
  ]
 }
}
summary {
 systemName=NS_Emote_Steep_Pour
 emitters [
  {
   bEnabled=false
   emitterName=Drip
   rendererClasses [&NiagaraSpriteRendererProperties, &NiagaraMeshRendererProperties]
  }
  {emitterName=TeaStream, rendererClasses [&NiagaraRibbonRendererProperties, &NiagaraSpriteRendererProperties]}
  {
   bEnabled=false
   emitterName=Syrup_Ribbons001
   rendererClasses [&NiagaraRibbonRendererProperties, &NiagaraSpriteRendererProperties]
  }
  {
   bEnabled=false
   emitterName=Syrup_Ribbons002
   rendererClasses [&NiagaraRibbonRendererProperties, &NiagaraSpriteRendererProperties]
  }
 ]
 userVariables [
  User.bIsFrontEnd {type=NiagaraBool}
  User.bIsFrontEndPreview {type=NiagaraBool}
  User.IsOnGround {type=NiagaraBool}
  User.RibbonIDOffset {type=NiagaraInt32}
  User.SkeletalMesh {
   type=NiagaraDataInterfaceSkeletalMesh
   defaultValue {
    struct=NiagaraExt_VariableValue_DataInterface
    value {dataInterface=&self:NiagaraDataInterfaceSkeletalMesh_0}
   }
  }
 ]
}
