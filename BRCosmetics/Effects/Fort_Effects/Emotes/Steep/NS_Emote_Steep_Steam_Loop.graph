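# Looping Niagara sprite steam effect for the Steep emote, apparently rising from a cup.
# One emitter spawns translucent puffs by time and by distance moved, with slight upward drift, drag and capped inherited velocity so they partly follow motion.
# Puffs grow over life with SubUV animation and age-driven material parameters.
# Alpha follows a life curve scaled by randomness and distance from the system's owner, fading puffs left behind.
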NiagaraSystem /BRCosmetics/Effects/Fort_Effects/Emotes/Steep/NS_Emote_Steep_Steam_Loop
dependencies {
 usedDataInterfaces [&NiagaraDataInterfaceCurve, &NiagaraDataInterfaceColorCurve]
 usedDynamicInputs []
 usedModules [&SystemState]
}
emitters {Steam {
  emitterUpdateScript [
   &EmitterState(Life Cycle Mode=::NewEnumerator1, Loop Duration=1, UseLoopDelay=true)
   &SpawnRate(SpawnRate=8, Use Spawn Probability=false)
   &SpawnPerUnit(Module.Use Max Movement Threshold=true, Module.Use Movement Tolerance=false, Spawn Per Unit=3, Use Spawn Probability=false)
  ]
  particleSpawnScript [
   &InitializeParticle(
    Color=<rgba 1 1 1 0.2>
    Lifetime Max=1.5
    Lifetime Min=0.6
    Lifetime Mode=::NewEnumerator1
    Position Offset=<0 0 4>
    Sprite Rotation Mode=::NewEnumerator2
    Sprite Size Mode=::NewEnumerator1
    Uniform Sprite Size Max=35
    Uniform Sprite Size Min=15
    UsePositionOffset=true
    Write Material Random=true
   )
   &AddVelocity(Velocity=&RandomRangeVector(Minimum=<-5 -5 10>, Maximum=<5 5 20>, Evaluation Type=::NewEnumerator1))
   &InheritVelocity(Module.Limit Speed=true, Velocity Limit=5, Velocity Scale=<0.25 0.25 0.25>)
  ]
  particleUpdateScript [
   &ParticleState()
   /SetVariables(Transient.DistanceBeforeFade=15) {isSetParameters=true}
   /SetVariables(Transient.DistanceFromCup:NiagaraFloat=&OneMinusFloat(Float=&ClampFloat(Float=&Multiply_Float(
       A=&Subtract_Float(
        A=&VectorLength(Vector=&Subtract_Vector(
          A=Linked() {linkedVariable {name=Particles.Position, type=NiagaraPosition}}
          B=Linked() {linkedVariable {name=Engine.Owner.Position, type=NiagaraPosition}}
         ))
        B=$Transient.DistanceBeforeFade
       )
       B=0.025
      )))) {isSetParameters=true}
   &SubUVAnimation(
    Sub UV Play Rate:NiagaraFloat=&RandomRangeFloat(Minimum=0.05, Maximum=0.1)
    SubUV Animation Mode:ENiagara_SubUVLookupMode=::NewEnumerator3
    SubUV Start Time Offset:NiagaraFloat=&RandomRangeFloat(Minimum=0, Maximum=63)
   )
   &DynamicMaterialParameters(
    Index 0 Param 1=&FloatFromCurve(FloatCurve=DataInterface() {propertyValues {Curve {keys [<at 0 -0.25>,<at 1 -1>]}}})
    Index 0 Param 4=&OneMinusFloat(Float=$Particles.NormalizedAge)
    Module.Param0WriteEnabled=true
    Module.Param1WriteEnabled=false
    Module.Param2WriteEnabled=false
    Module.Param3WriteEnabled=true
   )
   &ScaleColor(
    Scale Alpha=&FloatFromCurve(
     FloatCurve=DataInterface() {propertyValues {Curve {keys [
         {arriveTangent=5.2427, interpMode=RCIM_Cubic, leaveTangent=5.2427, tangentMode=RCTM_User, time=0, value=0}
         {arriveTangent=-1.0898, interpMode=RCIM_Cubic, leaveTangent=-1.0898, tangentMode=RCTM_User, time=0.3, value=1}
         {arriveTangent=-0.1231, interpMode=RCIM_Cubic, leaveTangent=-0.1231, tangentMode=RCTM_User, time=1, value=0.5}
        ]}}}
     Scale Curve=&Multiply_Float(A=&RandomRangeFloat(Minimum=0.5, Maximum=1), B=$Transient.DistanceFromCup)
    )
    ScaleA=true
    ScaleRGB=false
   )
   &ScaleSpriteSize(Scale Factor=&Vector2DFromFloat(Value=&FloatFromCurve(FloatCurve=DataInterface() {propertyValues {Curve {keys [
          {arriveTangent=1.8751, interpMode=RCIM_Cubic, leaveTangent=1.8751, tangentMode=RCTM_User, time=0, value=0.25}
          {arriveTangent=0.72, interpMode=RCIM_Cubic, leaveTangent=0.72, tangentMode=RCTM_User, time=1, value=1.25}
         ]}}})))
   &Drag(Drag=0.5, Use Linear Drag=true, Use Rotational Drag=false)
   &InheritVelocity(Module.Limit Speed=true, Velocity Limit=0.3, Velocity Scale=&VectorFromFloat(Value=0.01)) {module=InheritVelocity001}
   &SolveForcesAndVelocity(Clamp Velocity=false, Limit Acceleration=false)
   &Color(Color=&ColorFromCurve(
     ColorCurve=DataInterface() {propertyValues {
       AlphaCurve {keys [<at 0 1 RCIM_Cubic>]}
       BlueCurve {keys [<at 0 0 RCIM_Cubic>,<at 0.5 0>,<at 1 1 RCIM_Cubic>]}
       GreenCurve {keys [<at 0 0 RCIM_Cubic>,<at 0.5 1>,<at 1 0 RCIM_Cubic>]}
       RedCurve {keys [<at 0 1 RCIM_Cubic>,<at 0.5 0>,<at 1 0 RCIM_Cubic>]}
      }}
     CurveIndex=$Particles.DistanceFromCup
    )) {enabled=false}
  ]
  rendererClasses [&NiagaraSpriteRendererProperties]
  renderers [{rendererClass=&NiagaraSpriteRendererProperties}]
 }}
summary {
 systemName=NS_Emote_Steep_Steam_Loop
 emitters [{emitterName=Steam, rendererClasses [&NiagaraSpriteRendererProperties]}]
 userVariables []
}
