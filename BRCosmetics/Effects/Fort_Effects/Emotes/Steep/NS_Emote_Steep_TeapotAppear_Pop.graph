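# Niagara effect for the Steep emote, by its name the pop shown as the teapot appears: a short smoke puff with drifting sparkles.
# Smoke bursts once, expands against drag and fades, playing flipbook frames from random starts while curves drive dynamic material parameters.
# SparklesDrift adds a slightly delayed burst plus a low spawn rate from a small sphere, drifting outward with blinking alpha, bright warm HDR color and late shrink.
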NiagaraSystem /BRCosmetics/Effects/Fort_Effects/Emotes/Steep/NS_Emote_Steep_TeapotAppear_Pop
dependencies {
 usedDataInterfaces [&NiagaraDataInterfaceCurve, &NiagaraDataInterfaceColorCurve]
 usedDynamicInputs []
 usedModules [&SystemState]
}
emitters {
 Smoke {
  emitterUpdateScript [
   &EmitterState()
   &SpawnBurst_Instantaneous(Spawn Count=7, Use Spawn Probability=false) {module=SpawnBurst_Instantaneous001}
  ]
  particleSpawnScript [
   &InitializeParticle(
    Color Maximum=<rgba 0.5255 0.5255 0.5255 0.75>
    Color Minimum=<rgba 0.4323 0.4323 0.4323 0.25>
    Color Mode=::NewEnumerator2
    Lifetime Max=0.4
    Lifetime Min=0.25
    Lifetime Mode=::NewEnumerator1
    Sprite Rotation Mode=::NewEnumerator2
    Sprite Size Mode=::NewEnumerator1
    Uniform Sprite Size Max=50
    Uniform Sprite Size Min=25
    UsePositionOffset=false
    Write Material Random=true
   )
   &AccelerationForce(Acceleration=&UniformRangedVector(Minimum=<0 0 0>, Maximum=<0 0 50>))
   &AddVelocity(Velocity=&UniformRangedVector(Minimum=<-50 -50 -50>, Maximum=<50 50 50>))
   &ApplyInitialForces()
  ]
  particleUpdateScript [
   &ParticleState()
   &Drag(Drag=5, Use Linear Drag=true, Use Rotational Drag=false)
   &SolveForcesAndVelocity(Clamp Velocity=false, Limit Acceleration=false)
   &ScaleSpriteSize(Scale Factor=&Vector2DFromFloat(Value=&FloatFromCurve(FloatCurve=DataInterface() {propertyValues {Curve {keys [
          {arriveTangent=1.2787, interpMode=RCIM_Cubic, leaveTangent=1.2787, leaveTangentWeight=0.5984, tangentMode=RCTM_User, time=0, value=0.1}
          <at 1 1 RCIM_Cubic>
         ]}}})))
   /SetVariables(Particles.StartFrame:NiagaraFloat=&RandomRangeFloat(Minimum=0, Maximum=63)) {isSetParameters=true}
   /SetVariables(Particles.SubImageIndex=&Lerp_Float(
     A=$Particles.StartFrame
     B=&Add_Float(A=$Particles.StartFrame, B=&RandomRangeFloat(Minimum=8, Maximum=15))
     Alpha=$Particles.NormalizedAge
    )) {isSetParameters=true}
   &DynamicMaterialParameters(
    Index 0 Param 1=&FloatFromCurve(FloatCurve=DataInterface() {propertyValues {Curve {keys [{arriveTangent=-2.1023, interpMode=RCIM_Cubic, leaveTangent=-2.1023, tangentMode=RCTM_User, time=-0.0046, value=1.0113}, <at 1.0011 0.002>]}}})
    Index 0 Param 4=&FloatFromCurve(FloatCurve=DataInterface() {propertyValues {Curve {keys [<at 0 0 RCIM_Cubic>,<at 1 0.5>]}}})
    Module.Param0WriteEnabled=true
    Module.Param1WriteEnabled=false
    Module.Param2WriteEnabled=false
    Module.Param3WriteEnabled=true
   ) {module=DynamicMatParams}
   &CameraOffset(Camera Offset Amount=10)
   &ScaleColor(
    Scale Alpha=&FloatFromCurve(FloatCurve=DataInterface() {propertyValues {Curve {keys [
         {arriveTangent=-3.1757, interpMode=RCIM_Cubic, leaveTangent=-3.1757, tangentMode=RCTM_User, time=0, value=1}
         {arriveTangent=-0.2376, interpMode=RCIM_Cubic, leaveTangent=-0.2376, tangentMode=RCTM_User, time=1, value=0}
        ]}}})
    ScaleA=true
    ScaleRGB=false
   )
  ]
  rendererClasses [&NiagaraSpriteRendererProperties]
  renderers [{rendererClass=&NiagaraSpriteRendererProperties}]
 }
 SparklesDrift {
  emitterUpdateScript [
   &EmitterState()
   &SpawnBurst_Instantaneous(Spawn Count=10, Spawn Time=0.05, Use Spawn Probability=false)
   &SpawnRate(SpawnRate=5, Use Spawn Probability=false)
  ]
  particleSpawnScript [
   /Niagara/Modules/Spawn/Initialization/InitializeParticle(
    Color=<rgba 0.25 0.25 1 1>
    Lifetime=&UniformRangedFloat(Minimum=0.1, Maximum=0.6)
    Mesh Scale=<1.5 1.5 1.5>
    Module.Write Color=true
    Module.Write Lifetime=true
    Module.Write Mass=false
    Module.Write Material Random=true
    Module.Write Position=false
    Module.Write Scale=false
    Module.Write SpriteRotation=true
    Module.Write SpriteSize=true
    Position=&SimulationPosition()
    Sprite Rotation=&UniformRangedFloat(Minimum=0, Maximum=1)
    Sprite Size=&UniformRangedVector2D(
     Minimum=&Vector2DFromFloat(Value=5)
     Maximum=&Vector2DFromFloat(Value=8)
    )
   )
   /Niagara/Modules/Spawn/Location/SphereLocation(MaskSpawnGroup=false, Module.Override Randomness=false, Sphere Radius=10, Surface Only=false) {module=SphereLocation001}
   &AddVelocityFromPoint(Module.Use Falloff Distance=false, Velocity Strength=15)
   &CameraOffset(Camera Offset Amount=5)
  ]
  particleUpdateScript [
   &ParticleState()
   &Color(Color=&ColorFromCurve(ColorCurve=DataInterface() {propertyValues {
       AlphaCurve {keys [
         <at -0.0023 0 RCIM_Cubic>
         {arriveTangent=0.2428, interpMode=RCIM_Cubic, leaveTangent=0.2428, time=0.1, value=1}
         {arriveTangent=-0.0163, interpMode=RCIM_Cubic, leaveTangent=-0.0163, time=0.2779, value=0.0681}
         {arriveTangent=-0.0735, interpMode=RCIM_Cubic, leaveTangent=-0.0735, time=0.4489, value=0.9943}
         {arriveTangent=0.0148, interpMode=RCIM_Cubic, leaveTangent=0.0148, time=0.6472, value=0.0409}
         {arriveTangent=-0.1252, interpMode=RCIM_Cubic, leaveTangent=-0.1252, time=0.8337, value=1}
         <at 0.9977 -0.003 RCIM_Cubic>
        ]}
       BlueCurve {keys [<at -0.0023 1 RCIM_Cubic>,<at 0.448 3.2>,<at 0.9977 0.8 RCIM_Cubic>]}
       GreenCurve {keys [<at -0.0023 3.6667 RCIM_Cubic>,<at 0.448 6.4>,<at 0.9977 3.2 RCIM_Cubic>]}
       RedCurve {keys [<at -0.0023 5 RCIM_Cubic>,<at 0.448 8>,<at 0.9977 4 RCIM_Cubic>]}
      }}))
   &Drag(Drag=5, Use Linear Drag=true, Use Rotational Drag=false)
   &SolveForcesAndVelocity(Clamp Velocity=false, Limit Acceleration=false)
   &ScaleSpriteSize(Scale Factor=&Vector2DFromFloat(Value=&FloatFromCurve(FloatCurve=DataInterface() {propertyValues {Curve {keys [<at 0.5 1>,<at 1 0>]}}})))
  ]
  rendererClasses [&NiagaraSpriteRendererProperties]
  renderers [{rendererClass=&NiagaraSpriteRendererProperties}]
 }
}
summary {
 systemName=NS_Emote_Steep_TeapotAppear_Pop
 emitters [
  {emitterName=Smoke, rendererClasses [&NiagaraSpriteRendererProperties]}
  {emitterName=SparklesDrift, rendererClasses [&NiagaraSpriteRendererProperties]}
 ]
 userVariables []
}
