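# Niagara one-shot sprite burst for the "pop" of the Steep emote's teapot resolve.
# Core spawns one spinning flash that swells and shrinks, with spawn-group masks left for a disabled second burst.
# Smoke puffs advance through random-start subimage frames, while delayed sparkle bursts add upward curl-noise swirls and twinkling drift via an oscillating alpha curve.
# Smoke and sparkles are thrown along +X and slowed by drag; Pop_Embers is disabled.
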NiagaraSystem /BRCosmetics/Effects/Fort_Effects/Emotes/Steep/NS_Emote_Steep_TeapotResolve_Pop
dependencies {
 usedDataInterfaces [&NiagaraDataInterfaceCurve, &NiagaraDataInterfaceColorCurve]
 usedDynamicInputs []
 usedModules [&SystemState]
}
emitters {
 Core {
  emitterUpdateScript [
   &EmitterState()
   &SpawnBurst_Instantaneous(Spawn Count=1, Use Spawn Probability=false)
   &SpawnBurst_Instantaneous(Spawn Count=1, Spawn Group=1, Use Spawn Probability=false) {enabled=false, module=SpawnBurst_Instantaneous001}
  ]
  particleSpawnScript [
   &InitializeParticle(
    Color=<rgba 1 0.7662 0.5 1>
    Lifetime=&MaskFloatBySpawnGroup(Masked Float=0.2, Passthrough Float=0.4)
    Sprite Rotation Mode=::NewEnumerator2
    Sprite Size Mode=::NewEnumerator3
    Uniform Sprite Size=&MaskFloatBySpawnGroup(Masked Float=50, Passthrough Float=25)
    UsePositionOffset=false
    Write Material Random=true
   )
   &AddVelocity(Velocity=<30 -10 -10>)
  ]
  particleUpdateScript [
   &Drag(Drag=3, Use Linear Drag=true, Use Rotational Drag=false)
   &SolveForcesAndVelocity(Clamp Velocity=false, Limit Acceleration=false)
   &ParticleState()
   &DynamicMaterialParameters(
    Index 0 Param 1=&FloatFromCurve(FloatCurve=DataInterface() {propertyValues {Curve {keys [<at 0 0 RCIM_Cubic>,<at 1 1 RCIM_Cubic>]}}})
    Index 0 Param 2=50
    Index 0 Param 3=1
    Index 0 Param 4=1
    Module.Param0WriteEnabled=true
    Module.Param1WriteEnabled=true
    Module.Param2WriteEnabled=true
    Module.Param3WriteEnabled=true
   )
   &ScaleColor(
    Scale Alpha=&FloatFromCurve(
     FloatCurve=DataInterface() {propertyValues {Curve {keys [<at 0.3 0.95 RCIM_Cubic>,<at 1 0>]}}}
     Scale Curve=&MaskFloatBySpawnGroup(Masked Float=1, Passthrough Float=0.2)
    )
    ScaleA=true
    ScaleRGB=true
   )
   &ScaleSpriteSize(Scale Factor=&Vector2DFromFloat(Value=&FloatFromCurve(FloatCurve=DataInterface() {propertyValues {Curve {keys [
          {interpMode=RCIM_Cubic, tangentMode=RCTM_User, time=0, value=0}
          {arriveTangent=-2.9126, interpMode=RCIM_Cubic, leaveTangent=-2.9126, tangentMode=RCTM_User, time=0.2, value=1}
          {arriveTangent=-0.1238, interpMode=RCIM_Cubic, leaveTangent=-0.1238, tangentMode=RCTM_User, time=1, value=0.2}
         ]}}})))
   /SetVariables(Particles.SpriteSize=&MaskVector2dBySpawnGroup(Masked Vector2D=$Particles.SpriteSize, Passthrough Vector2D=$Particles.Initial.SpriteSize, Spawn Group Mask=0)) {isSetParameters=true}
   &SpriteRotationRate(Rotation Rate=&FloatFromCurve(
     FloatCurve=DataInterface() {propertyValues {Curve {keys [
         {interpMode=RCIM_Cubic, tangentMode=RCTM_User, time=0, value=1}
         {arriveTangent=-1.5604, interpMode=RCIM_Cubic, leaveTangent=-1.5604, tangentMode=RCTM_User, time=0.3, value=0.375}
         {arriveTangent=-0.5573, interpMode=RCIM_Cubic, leaveTangent=-0.5573, tangentMode=RCTM_User, time=1, value=0.05}
        ]}}}
     Scale Curve=500
    ))
   /SetVariables(Particles.SpriteRotation:NiagaraFloat=&MaskFloatBySpawnGroup(
     Masked Float=$Particles.SpriteRotation
     Passthrough Float=&Add_Float(A=$Particles.Initial.SpriteRotation, B=&Multiply_Float(A=$Particles.NormalizedAge, B=30))
    )) {isSetParameters=true}
  ]
  rendererClasses [&NiagaraSpriteRendererProperties]
  renderers [{rendererClass=&NiagaraSpriteRendererProperties}]
 }
 Pop_Embers {
  bEnabled=false
  emitterUpdateScript [
   &EmitterState()
   &SpawnBurst_Instantaneous(Spawn Count=20, Use Spawn Probability=false)
   &SpawnRate(SpawnRate=50000, Use Spawn Probability=false) {enabled=false}
  ]
  particleSpawnScript [
   &InitializeParticle(
    Color Channel Mode:ENiagaraChannelCorrelation=::NewEnumerator2
    Color Maximum=<rgba 2 0.9733 0.1333 1>
    Color Minimum=<rgba 1 0.9067 0.0667 0.1>
    Color Mode=::NewEnumerator2
    Lifetime Max=0.6
    Lifetime Min=0.1
    Lifetime Mode=::NewEnumerator1
    Sprite Size Max=<3 6>
    Sprite Size Min=<1 2>
    Sprite Size Mode=::NewEnumerator2
    UsePositionOffset=false
    Write Material Random=true
   )
   /Niagara/Modules/Spawn/Location/SphereLocation(MaskSpawnGroup=false, Module.Override Randomness=false, Non Uniform Scale=<1 1 0.5>, Sphere Radius=10, Surface Only=false)
   &AddVelocity(Velocity=&Add_Vector(
     A=&RandomRangeVector(Minimum=<50 0 0>, Maximum=<200 0 0>)
     B=&Multiply_VectorByFloat(Vector=$Output.SphereLocation.SphereNormal, Float=&RandomRangeFloat(Minimum=25, Maximum=200))
    ))
  ]
  particleUpdateScript [
   &ParticleState()
   &AccelerationForce(Acceleration=<0 0 -100>) {enabled=false}
   &GravityForce(Gravity=<0 0 -500>)
   &Drag(Drag=4, Use Linear Drag=true, Use Rotational Drag=false)
   &SolveForcesAndVelocity(Clamp Velocity=false, Limit Acceleration=false)
   &ScaleColor(
    Scale Alpha=&FloatFromCurve(
     FloatCurve=DataInterface() {propertyValues {Curve {keys [
         {interpMode=RCIM_Cubic, tangentMode=RCTM_User, time=0, value=0}
         {interpMode=RCIM_Cubic, tangentMode=RCTM_User, time=0.5, value=1}
         {tangentMode=RCTM_User, time=1, value=0}
        ]}}}
     Scale Curve=&RandomRangeFloat(Minimum=0, Maximum=1, Evaluation Type=::NewEnumerator1)
    )
    Scale RGB=&Lerp_Vector(A=<1 1 1>, B=<1 0.2 0.2>, Alpha=$Particles.NormalizedAge)
    ScaleA=true
    ScaleRGB=true
   )
   &ScaleColor(
    Scale Alpha=&Power(a=&RandomRangeFloat(Minimum=0.2, Maximum=1, Evaluation Type=::NewEnumerator1), b=3)
    Scale RGB=&VectorFromFloat(Value=&RandomRangeFloat(Minimum=0, Maximum=5, Evaluation Type=::NewEnumerator1))
    ScaleA=true
    ScaleRGB=true
   ) {enabled=false, module=ScaleColor001}
   &ScaleSpriteSize(Scale Factor=&Lerp_Vector2(A=<1 1>, B=<0 0.5>, Alpha=$Particles.NormalizedAge))
  ]
  rendererClasses [&NiagaraSpriteRendererProperties]
  renderers [{rendererClass=&NiagaraSpriteRendererProperties}]
 }
 Smoke {
  emitterUpdateScript [
   &EmitterState()
   &SpawnBurst_Instantaneous(Spawn Count=7, Use Spawn Probability=false) {module=SpawnBurst_Instantaneous001}
  ]
  particleSpawnScript [
   &InitializeParticle(
    Color Maximum=<rgba 0.5255 0.5255 0.5255 0.75>
    Color Minimum=<rgba 0.4323 0.4323 0.4323 0.25>
    Color Mode=::NewEnumerator2
    Lifetime Max=0.4
    Lifetime Min=0.25
    Lifetime Mode=::NewEnumerator1
    Sprite Rotation Mode=::NewEnumerator2
    Sprite Size Mode=::NewEnumerator1
    Uniform Sprite Size Max=50
    Uniform Sprite Size Min=25
    UsePositionOffset=false
    Write Material Random=true
   )
   &AccelerationForce(Acceleration=&UniformRangedVector(Minimum=<0 0 0>, Maximum=<0 0 50>))
   &AddVelocity(Velocity=&Add_Vector(
     A=&RandomRangeVector(Minimum=<50 0 0>, Maximum=<200 0 0>)
     B=&UniformRangedVector(Minimum=<-50 -50 -50>, Maximum=<50 50 50>)
    ))
   &ApplyInitialForces()
  ]
  particleUpdateScript [
   &ParticleState()
   &Drag(Drag=5, Use Linear Drag=true, Use Rotational Drag=false)
   &SolveForcesAndVelocity(Clamp Velocity=false, Limit Acceleration=false)
   &ScaleSpriteSize(Scale Factor=&Vector2DFromFloat(Value=&FloatFromCurve(FloatCurve=DataInterface() {propertyValues {Curve {keys [
          {arriveTangent=1.2787, interpMode=RCIM_Cubic, leaveTangent=1.2787, leaveTangentWeight=0.5984, tangentMode=RCTM_User, time=0, value=0.1}
          <at 1 1 RCIM_Cubic>
         ]}}})))
   /SetVariables(Particles.StartFrame:NiagaraFloat=&RandomRangeFloat(Minimum=0, Maximum=63)) {isSetParameters=true}
   /SetVariables(Particles.SubImageIndex=&Lerp_Float(
     A=$Particles.StartFrame
     B=&Add_Float(A=$Particles.StartFrame, B=&RandomRangeFloat(Minimum=8, Maximum=15))
     Alpha=$Particles.NormalizedAge
    )) {isSetParameters=true}
   &DynamicMaterialParameters(
    Index 0 Param 1=&FloatFromCurve(FloatCurve=DataInterface() {propertyValues {Curve {keys [{arriveTangent=-2.1023, interpMode=RCIM_Cubic, leaveTangent=-2.1023, tangentMode=RCTM_User, time=-0.0046, value=1.0113}, <at 1.0011 0.002>]}}})
    Index 0 Param 4=&FloatFromCurve(FloatCurve=DataInterface() {propertyValues {Curve {keys [<at 0 0 RCIM_Cubic>,<at 1 0.5>]}}})
    Module.Param0WriteEnabled=true
    Module.Param1WriteEnabled=false
    Module.Param2WriteEnabled=false
    Module.Param3WriteEnabled=true
   ) {module=DynamicMatParams}
   &CameraOffset(Camera Offset Amount=10)
   &ScaleColor(
    Scale Alpha=&FloatFromCurve(FloatCurve=DataInterface() {propertyValues {Curve {keys [
         {arriveTangent=-3.1757, interpMode=RCIM_Cubic, leaveTangent=-3.1757, tangentMode=RCTM_User, time=0, value=1}
         {arriveTangent=-0.2376, interpMode=RCIM_Cubic, leaveTangent=-0.2376, tangentMode=RCTM_User, time=1, value=0}
        ]}}})
    ScaleA=true
    ScaleRGB=false
   )
  ]
  rendererClasses [&NiagaraSpriteRendererProperties]
  renderers [{rendererClass=&NiagaraSpriteRendererProperties}]
 }
 SparklesCurl {
  emitterUpdateScript [
   &EmitterState()
   &SpawnBurst_Instantaneous(Spawn Count=10, Spawn Time=0.05, Use Spawn Probability=false)
  ]
  particleSpawnScript [
   /Niagara/Modules/Spawn/Initialization/InitializeParticle(
    Color=<rgba 5 4.3125 2.25 1>
    Lifetime=&UniformRangedFloat(Minimum=0.25, Maximum=1)
    Module.Write Color=true
    Module.Write Lifetime=true
    Module.Write Mass=true
    Module.Write Material Random=true
    Module.Write Position=false
    Module.Write Scale=false
    Module.Write SpriteRotation=false
    Module.Write SpriteSize=true
    Position=&SimulationPosition()
    Sprite Size=&Vector2DFromFloat(Value=&UniformRangedFloat(Minimum=5, Maximum=10))
   )
   &ApplyInitialForces()
   /Niagara/Modules/Spawn/Location/V2/SphereLocation(
    MaskSpawnGroup=false
    Override Seed=false
    Sphere Radius=5
    Surface Only=true
    UseNonUniformScale=false
    UseOffset=false
    UseSphereOrientation=false
   )
   &AddVelocityFromPoint(Module.Use Falloff Distance=false, Velocity Strength=&RandomRangeFloat(Minimum=50, Maximum=150))
   &AddVelocity(Velocity=&RandomRangeVector(Minimum=<25 0 0>, Maximum=<400 0 0>))
  ]
  particleUpdateScript [
   &ParticleState()
   &GravityForce(Gravity=<0 0 100>)
   /Niagara/Modules/Update/Forces/CurlNoiseForce(Noise Frequency=0.04, Noise Strength=200, Pan Noise=false)
   &Drag(Drag=7, Use Linear Drag=true, Use Rotational Drag=false)
   &SolveForcesAndVelocity(Clamp Velocity=false, Limit Acceleration=false)
   &ScaleColor(
    Scale Alpha=&FloatFromCurve(FloatCurve=DataInterface() {propertyValues {Curve {keys [<at -0.0092 0.0144>,<at 0.0967 0.9767>,<at 1.0011 0.002>]}}})
    ScaleA=true
    ScaleRGB=true
   )
   &ScaleSpriteSize(Scale Factor=&Vector2DFromFloat(Value=&FloatFromCurve(FloatCurve=DataInterface() {propertyValues {Curve {keys [<at -0.0046 0>,<at 0.1746 0.9952>,<at 1.0011 0.002>]}}})))
  ]
  rendererClasses [&NiagaraSpriteRendererProperties]
  renderers [{rendererClass=&NiagaraSpriteRendererProperties}]
 }
 SparklesDrift {
  emitterUpdateScript [
   &EmitterState()
   &SpawnBurst_Instantaneous(Spawn Count=10, Spawn Time=0.05, Use Spawn Probability=false)
  ]
  particleSpawnScript [
   &InitializeParticle(
    Color=<rgba 0.25 0.25 1 1>
    Lifetime=&UniformRangedFloat(Minimum=0.1, Maximum=0.6)
    Mesh Scale=<1.5 1.5 1.5>
    Module.Write Color=true
    Module.Write Lifetime=true
    Module.Write Mass=false
    Module.Write Material Random=true
    Module.Write Position=false
    Module.Write Scale=false
    Module.Write SpriteRotation=true
    Module.Write SpriteSize=true
    Position=&SimulationPosition()
    Sprite Rotation=&UniformRangedFloat(Minimum=0, Maximum=1)
    Sprite Size=&UniformRangedVector2D(
     Minimum=&Vector2DFromFloat(Value=5)
     Maximum=&Vector2DFromFloat(Value=8)
    )
   )
   /Niagara/Modules/Spawn/Location/SphereLocation(MaskSpawnGroup=false, Module.Override Randomness=false, Sphere Radius=10, Surface Only=false) {module=SphereLocation001}
   &AddVelocityFromPoint(Module.Use Falloff Distance=false, Velocity Strength=15)
   &AddVelocity(Velocity=&RandomRangeVector(Minimum=<50 0 0>, Maximum=<400 0 0>))
   &CameraOffset(Camera Offset Amount=5)
  ]
  particleUpdateScript [
   &ParticleState()
   &Color(Color=&ColorFromCurve(ColorCurve=DataInterface() {propertyValues {
       AlphaCurve {keys [
         <at -0.0023 0 RCIM_Cubic>
         {arriveTangent=0.2428, interpMode=RCIM_Cubic, leaveTangent=0.2428, time=0.1, value=1}
         {arriveTangent=-0.0163, interpMode=RCIM_Cubic, leaveTangent=-0.0163, time=0.2779, value=0.0681}
         {arriveTangent=-0.0735, interpMode=RCIM_Cubic, leaveTangent=-0.0735, time=0.4489, value=0.9943}
         {arriveTangent=0.0148, interpMode=RCIM_Cubic, leaveTangent=0.0148, time=0.6472, value=0.0409}
         {arriveTangent=-0.1252, interpMode=RCIM_Cubic, leaveTangent=-0.1252, time=0.8337, value=1}
         <at 0.9977 -0.003 RCIM_Cubic>
        ]}
       BlueCurve {keys [<at -0.0023 1 RCIM_Cubic>,<at 0.448 3.2>,<at 0.9977 0.8 RCIM_Cubic>]}
       GreenCurve {keys [<at -0.0023 3.6667 RCIM_Cubic>,<at 0.448 6.4>,<at 0.9977 3.2 RCIM_Cubic>]}
       RedCurve {keys [<at -0.0023 5 RCIM_Cubic>,<at 0.448 8>,<at 0.9977 4 RCIM_Cubic>]}
      }}))
   &Drag(Drag=5, Use Linear Drag=true, Use Rotational Drag=false)
   &SolveForcesAndVelocity(Clamp Velocity=false, Limit Acceleration=false)
   &ScaleSpriteSize(Scale Factor=&Vector2DFromFloat(Value=&FloatFromCurve(FloatCurve=DataInterface() {propertyValues {Curve {keys [<at 0.5 1>,<at 1 0>]}}})))
  ]
  rendererClasses [&NiagaraSpriteRendererProperties]
  renderers [{rendererClass=&NiagaraSpriteRendererProperties}]
 }
}
summary {
 systemName=NS_Emote_Steep_TeapotResolve_Pop
 emitters [
  {emitterName=Core, rendererClasses [&NiagaraSpriteRendererProperties]}
  {bEnabled=false, emitterName=Pop_Embers, rendererClasses [&NiagaraSpriteRendererProperties]}
  {emitterName=Smoke, rendererClasses [&NiagaraSpriteRendererProperties]}
  {emitterName=SparklesCurl, rendererClasses [&NiagaraSpriteRendererProperties]}
  {emitterName=SparklesDrift, rendererClasses [&NiagaraSpriteRendererProperties]}
 ]
 userVariables []
}
